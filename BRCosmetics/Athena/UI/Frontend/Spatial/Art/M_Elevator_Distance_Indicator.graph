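# Translucent, unlit, two-sided material for an elevator distance indicator in the frontend's spatial UI.
# One tint drives a two-tone fill split along a rotated UV line, with an accent strip and an additive radial glow.
# A sine-timed pulse, gated by bIsBang, blends toward the bang color, while centered-UV edge masks fade opacity.
# State color parameters (Unavailable, Locked, Available, Completed, Bang) are present but unconnected.
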Material /BRCosmetics/Athena/UI/Frontend/Spatial/Art/M_Elevator_Distance_Indicator
graph {
 properties {
  blendMode=BLEND_Translucent
  numCustomizedUVs=1
  pixelDepthOffsetMode=PDOM_Legacy
  shadingModel=MSM_Unlit
  twoSided=true
  usedShadingModels=Unlit
 }
 nodes [
  n1 = VectorParameter() {parameterName=RGBA-1, defaultValue=<rgba 0.0452 0.0723 0.159 0.75>}
  n2 = Multiply(A=ScalarParameter() {defaultValue=1.5, parameterName=Color-Change}, B=@n1.RGB)
  n3 = MaterialFunctionCall(
   UVs=Multiply(A=TextureCoordinate(), B=Constant2Vector() {g=0.45, r=0.125})
   CenterPosition=Constant2Vector() {g=0.05, r=0.06}
   Radius=ScalarParameter() {defaultValue=0.1929, parameterName=Radius Scale}
   Density=ScalarParameter() {defaultValue=0.75, parameterName=RasiusDensity}
  ) {materialFunction=&RadialGradientExponential}
  n4 = ComponentMask(None=Rotator(Coordinate=TextureCoordinate(), Time=ScalarParameter() {defaultValue=-0.505, parameterName=Rotation}) {speed=6.2832}) {g=true}
  n5 = ScalarParameter() {defaultValue=0.7345, parameterName=Divide-Subtract}
  n6 = ScalarParameter() {defaultValue=128, parameterName=Divide-Edge}
  n7 = Saturate(None=Multiply(A=Subtract(A=@n4, B=@n5), B=@n6))
  n8 = ScalarParameter() {defaultValue=0.7345, parameterName=Divide-Accent-Subtract}
  n9 = Saturate(None=Multiply(A=Subtract(A=@n4, B=@n8), B=@n6))
  n10 = LinearInterpolate(A=@n7, B=ScalarParameter() {defaultValue=2, parameterName=Strip}, Alpha=DotProduct(A=OneMinus(None=@n9), B=@n7))
  n11 = Multiply(A=Time(), B=ScalarParameter() {parameterName=TimeFactor})
  n12 = Saturate(None=LinearInterpolate(A=ScalarParameter() {parameterName=Min}, B=ScalarParameter() {defaultValue=1, parameterName=Max}, Alpha=Sine(None=@n11)) {constA=0.25})
  n13 = ScalarParameter() {parameterName=bIsBang}
  LinearInterpolate(
   A=Add(A=Multiply(A=@n3.RadialGradientExponential, B=@n2), B=LinearInterpolate(A=@n1.RGB, B=@n2, Alpha=@n10))
   B=VectorParameter() {parameterName=RGBA-Bang, defaultValue=<rgba 1 1 0 1>}.RGB
   Alpha=Multiply(A=Multiply(A=@n12, B=@n13), B=ScalarParameter() {parameterName=Bang Multiply})
  )
  n14 = Add(A=TextureCoordinate()) {constB=-0.5}
  n15 = MaterialFunctionCall(Float2=Multiply(A=Abs(None=@n14)) {constB=2}) {materialFunction=&BreakOutFloat2Components}
  n16 = ScalarParameter() {defaultValue=0.25, parameterName=X-Edge Offset}
  n17 = Saturate(None=Multiply(A=Subtract(A=@n15.R, B=@n16), B=ScalarParameter() {defaultValue=2.1799, parameterName=X-Edge Sharpness}))
  n18 = ScalarParameter() {defaultValue=0.25, parameterName=Y-Edge Offset}
  n19 = Saturate(None=Multiply(A=Subtract(A=@n15.G, B=@n18), B=ScalarParameter() {defaultValue=2.1799, parameterName=Y-Edge Sharpness}))
  Multiply(A=@n1.A, B=OneMinus(None=LinearInterpolate(A=@n17, Alpha=@n19)))
  VectorParameter() {parameterName=Unavailable, defaultValue=<rgba 0.0452 0.0723 0.159 1>}
  VectorParameter() {parameterName=Locked, defaultValue=<rgba 0.2195 0.0137 0.0382 1>}
  VectorParameter() {parameterName=Available, defaultValue=<rgba 0.0052 0.0887 0.4287 1>}
  VectorParameter() {parameterName=Completed, defaultValue=<rgba 0.0044 0.2789 0.1384 1>}
  VectorParameter() {parameterName=Bang, defaultValue=<rgba 1 1 0 1>}
 ]
}
