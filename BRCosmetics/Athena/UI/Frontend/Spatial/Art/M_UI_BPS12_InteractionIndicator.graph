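# Translucent UI material for the BPS12 interaction indicator in the frontend spatial UI.
# It samples a linear-color indicator texture through a static component mask parameter, TextureMask, which selects the red channel by default.
# A Color vector parameter, blue by default, exposes the tint to material instances.
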Material /BRCosmetics/Athena/UI/Frontend/Spatial/Art/M_UI_BPS12_InteractionIndicator
graph {
 properties {blendMode=BLEND_Translucent, materialDomain=MD_UI, pixelDepthOffsetMode=PDOM_Legacy, usedShadingModels=Default Lit}
 nodes [
  StaticComponentMaskParameter(None=TextureSample() {samplerType=SAMPLERTYPE_LinearColor, texture=&/T_UI_BPS12_InteractionIndicator}.RGB) {defaultR=true, parameterName=TextureMask}
  VectorParameter() {parameterName=Color, defaultValue=<rgba 0 0.4709 1 1>}
 ]
}
