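# Master translucent UI material that draws the gradient for hue, saturation or value color sliders, selected by static switches.
# In hue mode it converts horizontal position on UV channel 4 into a full-saturation, full-brightness HSV rainbow.
# Otherwise it blends between two endpoints built from Hue, Saturation, Value and left/right RGB parameters, with black as the value slider's left end.
# A curve atlas row remaps that blend along the slider.
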Material /BRCosmetics/Athena/UI/Frontend/Spatial/Art/M_UI_ColorSliderMaster
graph {
 properties {blendMode=BLEND_Translucent, materialDomain=MD_UI, pixelDepthOffsetMode=PDOM_Legacy, usedShadingModels=Default Lit}
 nodes [
  n1 = ComponentMask(None=TextureCoordinate() {coordinateIndex=4}) {r=true}
  n2 = Constant() {r=1}
  n3 = Constant() {r=1}
  n4 = MaterialFunctionCall(Float3=MaterialFunctionCall!(RGB=VectorParameter() {parameterName=RGB_Left, defaultValue=<rgba 1 1 1 1>}.RGB) {materialFunction=/Engine/Functions/Engine_MaterialFunctions02/Math/RGBtoHSV}."RGB to HSV (0-1)") {materialFunction=&BreakOutFloat3Components}
  n5 = Divide(A=ScalarParameter() {defaultValue=100, parameterName=Value, sliderMax=100}, B=Constant() {r=100})
  n6 = MaterialFunctionCall!(HSV_In=MaterialFunctionCall(X=@n4.R, Y=@n4.R, Z=@n5) {materialFunction=&MakeFloat3}.Result) {materialFunction=/Game/Athena/UI/Materials/MF_HSVtoRGB}
  n7 = Constant()
  n8 = MaterialFunctionCall(Float3=MaterialFunctionCall(RGB=VectorParameter() {parameterName=RGB_Right, defaultValue=<rgba 1 0 0 1>}.RGB) {materialFunction=&RGBtoHSV}."RGB to HSV (0-1)") {materialFunction=&BreakOutFloat3Components}
  n9 = ScalarParameter() {defaultValue=100, parameterName=Saturation, sliderMax=100}
  n10 = Constant() {r=100}
  n11 = MaterialFunctionCall(
   X=Divide(A=ScalarParameter() {defaultValue=180, parameterName=Hue, sliderMax=360}, B=Constant() {r=360})
   Y=StaticSwitchParameter(True=@n8.G, False=Divide(A=@n9, B=@n10)) {parameterName=IsSaturationSlider}
   Z=StaticSwitchParameter(True=@n5, False=@n8.B) {parameterName=IsSaturationSlider}
  ) {materialFunction=&MakeFloat3}
  n12 = CurveAtlasRowParameter(CurveTime=@n1) {
   atlas=&/CA_ColorSlider
   curve=&/C_ColorSlider_Value
   parameterName=ReMapCurve
  }
  StaticSwitchParameter(
   True=MaterialFunctionCall(HSV_In=MaterialFunctionCall(X=@n1, Y=@n2, Z=@n3) {materialFunction=&MakeFloat3}.Result) {materialFunction=&MF_HSVtoRGB}.RGB_Out
   False=LinearInterpolate(
    A=StaticSwitchParameter(True=@n6.RGB_Out, False=@n7) {parameterName=IsSaturationSlider}
    B=MaterialFunctionCall(HSV_In=@n11.Result) {materialFunction=&MF_HSVtoRGB}.RGB_Out
    Alpha=@n12
   )
  ) {parameterName=IsHueSlider}
 ]
}
