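# UI master material for umbrella icons in the frontend, recolouring a channel-packed mask texture with three tint colours.
# The Icon mask's R, G and B channels weight the Primary, Secondary and Tertiary colour parameters, which are summed and saturated into a colour.
# A second saturated chain builds an alpha from the mask, scaled by OverallAlpha.
# That alpha sum counts R twice plus B, so the green channel does not contribute to it.
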Material /BRCosmetics/Athena/UI/Frontend/Spatial/Art/M_UI_UmbrellaIcon_Master
graph {
 properties {blendMode=BLEND_Translucent, materialDomain=MD_UI, pixelDepthOffsetMode=PDOM_Legacy, usedShadingModels=Default Lit}
 nodes [
  n1 = TextureSampleParameter2D(UVs=TextureCoordinate()) {parameterName=Icon, samplerType=SAMPLERTYPE_LinearColor, texture=&/T_UI_UmbrellaBuildIconMask_128x}
  n2 = Multiply(A=VectorParameter() {parameterName=Primary Color, defaultValue=<rgba 1 0 0 1>}.RGB, B=@n1.R)
  n3 = Multiply(A=VectorParameter() {parameterName=Secondary Color, defaultValue=<rgba 0 1 0 1>}.RGB, B=@n1.G)
  n4 = VectorParameter() {parameterName=Tertiary Color, defaultValue=<rgba 0 0 1 1>}
  Saturate(None=Add(A=Add(A=@n2, B=@n3), B=Multiply(A=@n4.RGB, B=@n1.B)))
  n5 = Add(A=Add(A=@n1.R, B=@n1.R), B=@n1.B)
  Saturate(None=Multiply(A=Saturate(None=@n5), B=ScalarParameter() {defaultValue=1, parameterName=OverallALpha}))
 ]
}
