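# Translucent UI master material for frontend item cards, drawing an icon over a radial gradient with locked and unlocked palettes.
# The icon is aspect-corrected, animatable in scale and opacity, optionally flipped, and faded vertically.
# Saturated, rotated UV edge masks, some resolution-corrected for sharpness, build the card shape, clip the icon and derive the border.
# Parameter lerps layer an animated divet, hover border color and optional side bar, plus overall opacity control.
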Material /BRCosmetics/Athena/UI/Frontend/Spatial/Art/M_Ui_Spatial_itemCard_Umbrella_Master
graph {
 properties {blendMode=BLEND_Translucent, materialDomain=MD_UI, pixelDepthOffsetMode=PDOM_Legacy, usedShadingModels=Default Lit}
 nodes [
  n1 = TextureCoordinate()
  n2 = LinearInterpolate(Alpha=ScalarParameter() {group=divet, parameterName=DivetAnim, sliderMax=1}) {constA=-45, constB=-90}
  n3 = Add(
   A=Rotator(Coordinate=@n1, Time=Divide(A=@n2) {constB=360}) {centerX=0, speed=6.2832}
   B=ScalarParameter() {defaultValue=-0.65, group=divet, parameterName=DivetOffset}
  )
  n4 = ScalarParameter() {defaultValue=64, group=Offset, parameterName=Sharpness}
  n5 = Saturate(None=Multiply(A=ComponentMask(None=@n3) {g=true}, B=@n4))
  n6 = Saturate(None=OneMinus(None=ScalarParameter() {defaultValue=0.75, group=Icon, parameterName=IconOpacity}))
  n7 = ScalarParameter() {group=Icon, parameterName=AnimateIconOpacity}
  n8 = MaterialFunctionCall() {materialFunction=&GetUserInterfaceUV}
  n9 = MaterialFunctionCall(Float2=@n8.Pixel Size) {materialFunction=&BreakOutFloat2Components}
  n10 = Max(A=@n9.G)
  n11 = Multiply(A=@n8.Normalized UV, B=AppendVector(A=Divide(A=@n9.R, B=@n10), B=Constant() {r=1}))
  n12 = ScalarParameter() {parameterName=UseSmallerIcon}
  n13 = LinearInterpolate(
   A=LinearInterpolate(
    A=ScalarParameter() {defaultValue=0.85, group=Icon, parameterName=IconScaleStart}
    B=ScalarParameter() {defaultValue=0.85, group=Icon, parameterName=IconScaleStartSmall}
    Alpha=@n12
   )
   B=LinearInterpolate(
    A=ScalarParameter() {defaultValue=0.95, group=Icon, parameterName=IconScaleStop}
    B=ScalarParameter() {defaultValue=0.95, group=Icon, parameterName=IconScaleStopSmall}
    Alpha=@n12
   )
   Alpha=ScalarParameter() {parameterName=AnimateIconScale}
  )
  n14 = LinearInterpolate(B=ScalarParameter() {defaultValue=-0.05, group=Icon, parameterName=Translate_Icon_U}, Alpha=ScalarParameter() {parameterName=UseBar})
  n15 = ScalarParameter() {group=Icon, parameterName=Translate_Icon_V}
  n16 = MaterialFunctionCall(Float2=Add(
    A=MaterialFunctionCall(UVs=@n11, Texture Scale=@n13) {materialFunction=&ScaleUVsByCenter}.UVs
    B=AppendVector(A=@n14, B=@n15)
   )) {materialFunction=&BreakOutFloat2Components}
  n17 = OneMinus(None=@n16.R)
  n18 = ScalarParameter() {parameterName=FlipPortrait}
  n19 = TextureSampleParameter2D(UVs=AppendVector(A=LinearInterpolate(A=@n16.R, B=@n17, Alpha=@n18), B=@n16.G)) {
   group=Icon
   parameterName=ItemIcon
   samplerSource=SSM_Clamp_WorldGroupSettings
   texture=/BRCosmetics/UI/Foundation/Textures/Icons/Heroes/Athena/Soldier/T-Soldier-HID-603-Athena-Commando-M-SpookyNeon
  }
  n20 = Subtract(A=@n19.A, B=Saturate(None=LinearInterpolate(A=@n6, Alpha=@n7) {constB=0}))
  n21 = LinearInterpolate(
   A=ScalarParameter() {defaultValue=-0.5, group=Icon, parameterName=Gradient_Fade_Falloff_Location}
   Alpha=ComponentMask(None=TextureCoordinate()) {g=true}
  )
  n22 = Saturate(None=Power(Base=Saturate(None=@n21), Exp=ScalarParameter() {defaultValue=2, group=Icon, parameterName=Gradient_Fade_Falloff_Rate}))
  n23 = Multiply(A=ScalarParameter() {defaultValue=0.025, group=Border, parameterName=Border_Left_Rotation}) {constB=0.01}
  n24 = TextureCoordinate()
  n25 = LinearInterpolate(
   A=ScalarParameter() {defaultValue=-10, group=Border, parameterName=Border_Left_Offset}
   B=ScalarParameter() {defaultValue=-20, group=Border, parameterName=Border_Left_Offset_WithBar}
   Alpha=ScalarParameter() {parameterName=UseBar}
  )
  n26 = ComponentMask(None=Add(A=Rotator(Coordinate=@n24, Time=@n23) {speed=6.2832}, B=Multiply(A=@n25) {constB=0.01})) {r=true}
  n27 = MaterialFunctionCall!(Native_U_PixelSize=ScalarParameter() {defaultValue=128, group=ResolutionSharpness, parameterName=NativePixelSize}) {materialFunction=/FortUILibrary/Foundation/Art_ToolKit/Tools/M_UI_Resolution_Sharpness_Correction}
  n28 = Multiply(A=@n27.Result, B=ScalarParameter() {defaultValue=128, group=Border, parameterName=Sharpness_U})
  n29 = Rotator(Coordinate=@n24, Time=Multiply(A=ScalarParameter() {defaultValue=0.025, group=Border, parameterName=Border_Right_Rotation}) {constB=0.01}) {speed=6.2832}
  n30 = ScalarParameter() {defaultValue=-10, group=Border, parameterName=Border_Right_Offset}
  n31 = ComponentMask(None=Add(A=OneMinus(None=@n29), B=Multiply(A=@n30) {constB=0.01})) {r=true}
  n32 = DotProduct(A=Saturate(None=Multiply(A=@n26, B=@n28) {constB=32}), B=Saturate(None=Multiply(A=@n31, B=@n28) {constB=32}))
  n33 = Multiply(A=ScalarParameter() {defaultValue=0.025, group=Border, parameterName=Border_Top_Rotation}) {constB=0.01}
  n34 = ScalarParameter() {defaultValue=-10, group=Border, parameterName=Border_Top_Offset}
  n35 = ComponentMask(None=Add(A=Rotator(Coordinate=@n24, Time=@n33) {speed=6.2832}, B=Multiply(A=@n34) {constB=0.01})) {g=true}
  n36 = Multiply(A=@n27.Result, B=ScalarParameter() {defaultValue=128, group=Border, parameterName=Sharpness_V})
  n37 = Rotator(Coordinate=@n24, Time=Multiply(A=ScalarParameter() {defaultValue=0.025, group=Border, parameterName=Border_Bottom_Rotation}) {constB=0.01}) {speed=6.2832}
  n38 = ScalarParameter() {defaultValue=-10, group=Border, parameterName=Border_Bottom_Offset}
  n39 = ComponentMask(None=Add(A=OneMinus(None=@n37), B=Multiply(A=@n38) {constB=0.01})) {g=true}
  n40 = DotProduct(A=Saturate(None=Multiply(A=@n35, B=@n36) {constB=32}), B=Saturate(None=Multiply(A=@n39, B=@n36) {constB=32}))
  n41 = Saturate(None=DotProduct(A=Saturate(None=@n32), B=Saturate(None=@n40)))
  n42 = DotProduct(A=Multiply(A=Saturate(None=@n20), B=OneMinus(None=@n22)), B=@n41)
  n43 = ComponentMask(None=Add(A=@n1, B=ScalarParameter() {defaultValue=-0.035, group=Offset, parameterName=TopOffset})) {g=true}
  n44 = ScalarParameter() {defaultValue=0.035, group=Offset, parameterName=BottomOffset}
  n45 = OneMinus(None=ComponentMask(None=Add(A=@n1, B=@n44)) {g=true})
  n46 = Multiply(A=Saturate(None=Multiply(A=@n43, B=@n4)), B=Saturate(None=Multiply(A=@n45, B=@n4)))
  n47 = ComponentMask(None=Add(A=@n1, B=ScalarParameter() {defaultValue=-0.035, group=Offset, parameterName=LeftOffset})) {r=true}
  n48 = ScalarParameter() {defaultValue=0.035, group=Offset, parameterName=RightOffset}
  n49 = OneMinus(None=ComponentMask(None=Add(A=@n1, B=@n48)) {r=true})
  n50 = Multiply(A=Saturate(None=Multiply(A=@n47, B=@n4)), B=Saturate(None=Multiply(A=@n49, B=@n4)))
  n51 = Saturate(None=Multiply(A=Saturate(None=@n46), B=Saturate(None=@n50)))
  n52 = LinearInterpolate(A=@n42, B=@n51, Alpha=ScalarParameter() {defaultValue=1, parameterName=UseBackground, sliderMax=1})
  DotProduct(A=@n5, B=@n52)
  n53 = VectorParameter() {group=Gradient, parameterName=LockedGradientStart, defaultValue=<rgba 0.0343 0.0467 0.0953 1>}
  n54 = VectorParameter() {group=Gradient, parameterName=UnlockedGradientStart, defaultValue=<rgba 0.0075 0.0176 0.2086 1>}
  n55 = ScalarParameter() {defaultValue=1, group=Gradient, parameterName=IsUnlocked}
  n56 = VectorParameter() {group=Gradient, parameterName=LockedGradientStop, defaultValue=<rgba 0.0782 0.1301 0.3712 1>}
  n57 = VectorParameter() {group=Gradient, parameterName=UnlockedGradientStop, defaultValue=<rgba 0 0.0908 1 1>}
  n58 = ScalarParameter() {defaultValue=0.6, group=Gradient, parameterName=GradientRadius}
  n59 = ScalarParameter() {defaultValue=1.2, group=Gradient, parameterName=GradientDensity}
  n60 = LinearInterpolate(
   A=LinearInterpolate(
    A=LinearInterpolate(A=@n53.RGB, B=@n54.RGB, Alpha=@n55)
    B=LinearInterpolate(A=@n56.RGB, B=@n57.RGB, Alpha=@n55)
    Alpha=MaterialFunctionCall(Radius=@n58, Density=@n59) {materialFunction=&RadialGradientExponential}.RadialGradientExponential
   )
   B=@n19.RGB
   Alpha=@n42
  )
  n61 = ScalarParameter() {defaultValue=1, parameterName=UseBackground, sliderMax=1}
  n62 = Add(A=@n44) {constB=0.005}
  n63 = OneMinus(None=ComponentMask(None=Add(A=@n1, B=@n62)) {g=true})
  n64 = DotProduct(A=@n5, B=Saturate(None=Multiply(A=@n63, B=@n4)))
  n65 = VectorParameter() {group=Gradient, parameterName=LockedBorder, defaultValue=<rgba 0.0901 0.1224 0.25 1>}
  n66 = LinearInterpolate(
   A=VectorParameter() {group=Gradient, parameterName=LockedBorderHover, defaultValue=<rgba 0.0901 0.1224 0.25 1>}.RGB
   B=VectorParameter() {group=Gradient, parameterName=UnlockedBorderHover, defaultValue=<rgba 0.0144 0.0786 0.4 1>}.RGB
   Alpha=ScalarParameter() {defaultValue=1, group=Gradient, parameterName=IsUnlocked}
  )
  n67 = ScalarParameter() {parameterName=AnimateBorder}
  n68 = VectorParameter() {group=Gradient, parameterName=UnlockedBorder, defaultValue=<rgba 0.0144 0.0786 0.4 1>}
  n69 = DotProduct(A=@n51, B=OneMinus(None=@n41))
  n70 = LinearInterpolate(
   A=LinearInterpolate(
    A=LinearInterpolate(A=@n19.RGB, B=@n60, Alpha=@n61)
    B=VectorParameter() {group=divet, parameterName=DivetColor, defaultValue=<rgba 0 0 0 1>}.RGB
    Alpha=Saturate(None=@n64)
   )
   B=LinearInterpolate(
    A=LinearInterpolate(A=@n65.RGB, B=@n66, Alpha=@n67)
    B=LinearInterpolate(A=@n68.RGB, B=@n66, Alpha=@n67)
    Alpha=ScalarParameter() {defaultValue=1, group=Gradient, parameterName=IsUnlocked}
   )
   Alpha=@n69
  )
  n71 = Multiply(A=ScalarParameter() {defaultValue=0.025, group=Border, parameterName=Bar_Right_Rotation}) {constB=0.01}
  n72 = ScalarParameter() {defaultValue=-15, group=Border, parameterName=Bar_Right_Offset}
  n73 = ComponentMask(None=Add(A=Rotator(Coordinate=@n24, Time=@n71) {speed=6.2832}, B=Multiply(A=@n72) {constB=0.01})) {r=true}
  n74 = Multiply(A=@n27.Result, B=ScalarParameter() {defaultValue=128, group=Border, parameterName=Sharpness_U})
  n75 = OneMinus(None=Saturate(None=Multiply(A=@n73, B=@n74) {constB=32}))
  LinearInterpolate(
   A=@n70
   B=LinearInterpolate(A=@n70, B=VectorParameter() {group=Gradient, parameterName=BarColor, defaultValue=<rgba 0 0.8333 1 1>}.RGB, Alpha=@n75)
   Alpha=ScalarParameter() {parameterName=UseBar}
  )
  n76 = Multiply(A=Clamp(None=ScalarParameter() {defaultValue=1, parameterName=OverallOpacity}), B=@n52)
  n77 = Saturate(None=DotProduct(A=@n75, B=@n69))
  LinearInterpolate(
   A=LinearInterpolate(A=@n76, B=LinearInterpolate(A=@n76, Alpha=@n77), Alpha=ScalarParameter() {parameterName=UseBar})
   B=@n52
   Alpha=ScalarParameter() {parameterName=BypassAlphaChange}
  )
 ]
}
