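# Translucent UI material that draws a circular color swatch with an inner disc and an outer ring for the frontend.
# Radial gradients multiplied by Sharpness and saturated form crisp circle masks, with separate inner and outer opacity controls combined into one alpha.
# AnimateInnerRadius grows the inner disc, blends the outer opacity between its min and max, and shifts the area outside the disc from the RGB1 color toward white.
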Material /BRCosmetics/Athena/UI/Frontend/Spatial/Art/M_Ui_circular_Swatch
graph {
 properties {blendMode=BLEND_Translucent, materialDomain=MD_UI, pixelDepthOffsetMode=PDOM_Legacy, usedShadingModels=Default Lit}
 nodes [
  n1 = VectorParameter() {group=Gradient, parameterName=RGB1, defaultValue=<rgba 0.0343 0.0467 0.0953 1>}
  n2 = LinearInterpolate(
   A=ScalarParameter() {defaultValue=0.4, group=Inner circle, parameterName=InnerRadius_Min, sliderMax=0.5}
   B=ScalarParameter() {defaultValue=0.48, group=Inner circle, parameterName=InnerRadius_Max, sliderMax=0.5}
   Alpha=ScalarParameter() {parameterName=AnimateInnerRadius, sliderMax=1}
  )
  n3 = ScalarParameter() {defaultValue=128, parameterName=Sharpness}
  n4 = Saturate(None=Multiply(
    A=MaterialFunctionCall(Radius=@n2) {materialFunction=&RadialGradientExponential}.RadialGradientExponential
    B=@n3
   ))
  n5 = OneMinus(None=@n4)
  LinearInterpolate(
   A=@n1.RGB
   B=LinearInterpolate(A=@n1.RGB, B=Constant() {r=1}, Alpha=ScalarParameter() {parameterName=AnimateInnerRadius, sliderMax=1})
   Alpha=@n5
  )
  n6 = MaterialFunctionCall() {materialFunction=&RadialGradientExponential}
  n7 = ScalarParameter() {defaultValue=0.75, group=Outer Circle, parameterName=Outer_MinOpacity, sliderMax=1}
  n8 = ScalarParameter() {defaultValue=0.75, group=Outer Circle, parameterName=Outer_MaxOpacity, sliderMax=1}
  n9 = ScalarParameter() {parameterName=AnimateInnerRadius}
  n10 = Subtract(A=Saturate(None=Multiply(A=@n6.RadialGradientExponential, B=@n3)), B=OneMinus(None=LinearInterpolate(A=@n7, B=@n8, Alpha=@n9)))
  n11 = Subtract(A=@n4, B=OneMinus(None=ScalarParameter() {defaultValue=1, group=Inner circle, parameterName=Inner_Opacity, sliderMax=1}))
  Saturate(None=Add(A=DotProduct(A=@n10, B=@n5), B=Saturate(None=@n11)))
 ]
}
