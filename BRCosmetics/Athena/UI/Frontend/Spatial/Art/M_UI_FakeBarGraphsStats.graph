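# Translucent UI material that animates a fake stat bar, apparently for frontend graph decoration.
# Time plus a per-instance offset is scaled, wrapped by Fmod and rounded, moving a fill edge along U in discrete steps before it resets.
# The inverted, TilingAmount-scaled U coordinate plus that step is multiplied by Softness and saturated into a mask, with no textures sampled.
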Material /BRCosmetics/Athena/UI/Frontend/Spatial/Art/M_UI_FakeBarGraphsStats
graph {
 properties {blendMode=BLEND_Translucent, materialDomain=MD_UI, pixelDepthOffsetMode=PDOM_Legacy, usedShadingModels=Default Lit}
 nodes [
  Constant() {r=1}
  n1 = ScalarParameter() {defaultValue=16, parameterName=TilingAmount}
  n2 = Multiply(A=TextureCoordinate(), B=AppendVector(A=@n1, B=Constant() {r=1}))
  n3 = Multiply(
   A=Add(A=Time(), B=ScalarParameter() {parameterName=RandomTimeOffset})
   B=ScalarParameter() {defaultValue=8, parameterName=GlobalTimeMultiplier}
  )
  n4 = ScalarParameter() {defaultValue=64, parameterName=BarOffset_Amount}
  n5 = Add(A=OneMinus(None=Add(A=@n2, B=@n1) {constB=16}), B=Round(None=Fmod(A=@n3, B=@n4)))
  Saturate(None=Multiply(A=ComponentMask(None=@n5) {r=true}, B=ScalarParameter() {defaultValue=4, parameterName=Softness}))
 ]
}
